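; Base Blueprint for the BoldUnicorn umbrella bouncer, which launches players and vehicles that enter its trigger volume.
; Pawns launch along an up-vector/hit-normal blend with fall-damage immunity, optional glider redeploy and a server quest stat.
; The server temporarily binds capsule hits so launched pawns damage buildings not tagged Skytube.Safe.
; Projectiles, C4 and physics balls get gravity scale flipped or offset on overlap, then adjusted back on exit.
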(event EventBeginPlay
  ;[1 1 1] Begin Play
  (Utilities|FlowControl|Sequence
    (:then_1
      (if (Variables|Default|GetAllowEffects)
        _))
    (:then_2
      (Class|FortAsyncActionAuthorityMatchReady|MatchReadyAsync
        (:OnReady
          (bind _outxy (DataTable|EvaluateCurveTableRow "/Game/Athena/Balance/DataTables/AthenaGameData.AthenaGameData" "Default.Vents.Disabled")
            (:RowFound
              (if (Variables|Default|GetisDisabled)
                _
                (elif (Math|Conversions|ToBoolean(Integer) (Math|Float|Truncate _outxy))
                  (Variables|Default|SetisDisabled true))))))))))

(event Custom|Building%20Hit%20Check
  (bind _outputdelegate (EventDispatchers|CreateEvent :!event Custom|PlayerCapsuleHit))
  (bind _fortpawns (Variables|Player|GetFortPawns))
  (bind _fortpawns_2 (Variables|Player|GetFortPawns))
  ;[1 1 1] "Break Buildings"
  (if (Networking|IsServer)
    (Collision|BindEventtoOnComponentHit (Variables|/Script/Engine.Character|GetCapsuleComponent (Variables|Player|GetFortPawn)) _outputdelegate)
    (bind _returnvalue (Utilities|Array|AddUnique _fortpawns (Variables|Player|GetFortPawn)))
    (Utilities|FlowControl|Sequence
      (:then_0
        (if (>= _returnvalue 0)
          (Utilities|Array|Add (Variables|Player|GetFortPawns_Timer) (Utilities|Time|GetGameTimeinSeconds))))
      (:then_1
        (Utilities|FlowControl|Delay (- (Variables|Default|GetKill%20Building%20Timer) (- (Utilities|Time|GetGameTimeinSeconds) (Utilities|Array|Get(acopy) (Variables|Player|GetFortPawns_Timer)))))
        (Collision|UnbindEventfromOnComponentHit (Variables|/Script/Engine.Character|GetCapsuleComponent (Utilities|Array|Get(acopy) _fortpawns)) _outputdelegate)
        (Utilities|Array|RemoveIndex _fortpawns_2)
        (Utilities|Array|RemoveIndex (Variables|Player|GetFortPawns_Timer))
        (if (Math|Conversions|ToBoolean(Integer) (Utilities|Array|Length _fortpawns_2))
          _)))))

(event OnComponentBeginOverlap(SM_SummerUmbrellaTriggerVolume)
  (params (OverlappedComponent :type Primitive%20Component%20Object%20Reference) (OtherActor :type Actor%20Object%20Reference) (OtherComp :type Primitive%20Component%20Object%20Reference) (OtherBodyIndex :type Integer) (bFromSweep :type Boolean) (SweepResult :type Hit%20Result%20Structure%20%28by%20ref%29))
  (bind (_bblockinghit _binitialoverlap _time _distance _location _impactpoint _normal _impactnormal) (Collision|BreakHitResult SweepResult))
  (bind _fortpawn (Variables|Player|GetFortPawn))
  ;[1 1 1] Begin Overlap
  (Audio|PlaySoundatLocation (Variables|Default|GetBounce%20SFX) (Transformation|GetWorldLocation (Variables|Default|GetSM_SummerUmbrellaTriggerVolume)))
  (Variables|Default|SetHit%20Normal _impactnormal)
  (bind _asfort_player_pawn (Utilities|Casting|CastToFortPlayerPawn OtherActor)
    (:then
      (Variables|Player|SetFortPawn _asfort_player_pawn)
      (if (Variables|Default|GetTrigger%20Glider%20Redeploy)
        (CallFunction|GliderRedeploy)
        (else
          _))
      (Character|LaunchCharacter _fortpawn (+ (* (Math|Vector|Normalize (Math|Vector|Lerp(Vector) (Transformation|GetActorUpVector) (* (Variables|Default|GetHit%20Normal) -1.0) 0.4)) (Variables|Default|GetCharacterLaunchVelocity)) (* (Math|Vector|Normalize (Transformation|GetVelocity (Variables|Player|GetFortPawn))) (Variables|Default|GetCharacter%20Forward%20Velocity))) (Variables|Default|GetXYOverride) (Variables|Default|GetZOverride))
      (Environment|ApplyGlobalEnvironmentGameplayEffecttoActor _fortpawn "/Game/Athena/Environments/Blueprints/DudeBro/GE_Athena_FallDmgImmune_FX.GE_Athena_FallDmgImmune_FX_C")
      (CallFunction|BuildingHitCheck)
      ;[1 1 1] Bounce off in Different Matches Quest hook in
      (if (Networking|IsServer)
        (Quest|SendComplexCustomStatEvent (Quests|GetQuestManager (Fortnite|GetFortPlayerControllerfromActor (Variables|Player|GetFortPawn)) "Invalid") 0 "(GameplayTags=((TagName=\"Athena.Quests.FDS.UmbrellaBounce\")))" "(GameplayTags=)" 1)))
    (:CastFailed
      (bind _asfort_athena_vehicle (Utilities|Casting|CastToFortAthenaVehicle OtherActor)
        (:then
          (Variables|Default|SetFortVehicle _asfort_athena_vehicle)
          (Vehicle|SetIgnoreNextFallingDamage (Variables|Default|GetFortVehicle) true)
          (Audio|BroadcastAbilityImpactatLocation 0 (Variables|Default|GetFortVehicle) (+ (Transformation|GetVelocity (Variables|Default|GetFortVehicle)) (* (Transformation|GetActorUpVector) (Variables|Default|GetCharacterLaunchVelocity))) 777.0))
        (:CastFailed
          (bind _asfort_projectile_base (Utilities|Casting|CastToFortProjectileBase OtherActor)
            (:then
              (CallFunction|ProjEnter _asfort_projectile_base))
            (:CastFailed
              (bind _asb_bga_athena_c4 (Utilities|Casting|CastToB_BGA_Athena_C4 OtherActor)
                (:then
                  (Variables|/Script/Engine.ProjectileMovementComponent|SetProjectileGravityScale (* (Variables|/Script/Engine.ProjectileMovementComponent|GetProjectileGravityScale (Variables|/Game/Athena/Items/Consumables/C4/B_BGA_Athena_C4.B_BGA_Athena_C4_C|GetProjectileMovement1 _asb_bga_athena_c4)) -1.0) (Variables|/Game/Athena/Items/Consumables/C4/B_BGA_Athena_C4.B_BGA_Athena_C4_C|GetProjectileMovement1 _asb_bga_athena_c4)))
                (:CastFailed
                  (bind _asphysics_ball_master (Utilities|Casting|CastToPhysicsBall_Master OtherActor))
                  (CallFunction|ToyEnter _asphysics_ball_master))))))))))

(event Custom|PlayerCapsuleHit
  (params (HitComponent :type Primitive%20Component%20Object%20Reference) (OtherActor :type Actor%20Object%20Reference) (OtherComp :type Primitive%20Component%20Object%20Reference) (NormalImpulse :type Vector) (Hit :type Hit%20Result%20Structure%20%28by%20ref%29))
  (bind (_bblockinghit _binitialoverlap _time _distance _location _impactpoint _normal _impactnormal) (Collision|BreakHitResult Hit))
  (bind _returnvalue (Components|GetOwner HitComponent))
  ;[1 1 1] "Break Buildings"
  (if (<= (.z _impactnormal) 0.0)
    (bind _asbuilding_actor (Utilities|Casting|CastToBuildingActor OtherActor))
    (Utilities|IsValid _returnvalue
      (:Is%20Valid
        (if (Variables|/Script/Engine.Actor|GetbCanBeDamaged _asbuilding_actor)
          (if (not (Actor|ActorHasTag _asbuilding_actor "Skytube.Safe"))
            (GameplayEffects|ApplyGameplayEffectToTarget (Ability|GetAbilitySystemComponent (Utilities|Casting|CastToFortPlayerPawn _returnvalue)) "/Game/Athena/Environments/Blueprints/SkyTube/GE_SkytubeDamageWallsLOTS.GE_SkytubeDamageWallsLOTS_C" (Ability|GetAbilitySystemComponent _asbuilding_actor) 0.0 (GameplayEffects|MakeEffectContext (Ability|GetAbilitySystemComponent (Utilities|Casting|CastToFortPlayerPawn _returnvalue))))))))))

(event OnComponentEndOverlap(SM_SummerUmbrellaTriggerVolume)
  (params (OverlappedComponent :type Primitive%20Component%20Object%20Reference) (OtherActor :type Actor%20Object%20Reference) (OtherComp :type Primitive%20Component%20Object%20Reference) (OtherBodyIndex :type Integer))
  ;[1 1 1] End Overlap
  (bind _asfort_player_pawn (Utilities|Casting|CastToFortPlayerPawn OtherActor)
    (:then
      (Variables|Player|SetFortPawn _asfort_player_pawn)
      (Ability|SendGameplayEventtoActor (Variables|Player|GetFortPawn) "(TagName=\"Abilities.Generic.Vent.Exit\")")
      (Variables|/Script/FortniteGame.FortPawn|SetbIgnoreNextFallingDamage true (Variables|Player|GetFortPawn)))
    (:CastFailed
      (bind _asfort_athena_vehicle (Utilities|Casting|CastToFortAthenaVehicle OtherActor)
        (:then
          (Variables|Default|SetFortVehicle _asfort_athena_vehicle)
          (Vehicle|SetIgnoreAllFallingDamage (Variables|Default|GetFortVehicle) false)
          (Vehicle|SetIgnoreNextFallingDamage (Variables|Default|GetFortVehicle) false))
        (:CastFailed
          (bind _asfort_projectile_base (Utilities|Casting|CastToFortProjectileBase OtherActor)
            (:then
              (CallFunction|ProjExit _asfort_projectile_base))
            (:CastFailed
              (bind _asb_bga_athena_c4 (Utilities|Casting|CastToB_BGA_Athena_C4 OtherActor)
                (:then
                  (Variables|/Script/Engine.ProjectileMovementComponent|SetProjectileGravityScale (* (Variables|/Script/Engine.ProjectileMovementComponent|GetProjectileGravityScale (Variables|/Game/Athena/Items/Consumables/C4/B_BGA_Athena_C4.B_BGA_Athena_C4_C|GetProjectileMovement1 _asb_bga_athena_c4)) -1.0) (Variables|/Game/Athena/Items/Consumables/C4/B_BGA_Athena_C4.B_BGA_Athena_C4_C|GetProjectileMovement1 _asb_bga_athena_c4)))
                (:CastFailed
                  (bind _asphysics_ball_master (Utilities|Casting|CastToPhysicsBall_Master OtherActor))
                  (CallFunction|ToyExit _asphysics_ball_master))))))))))

(fn GliderRedeploy
  (Ability|SendGameplayEventtoActor (Variables|Player|GetFortPawn) "(TagName=\"Abilities.Generic.Vent.Enter\")" (Utilities|Struct|MakeGameplayEventData :Instigator (Variables|Player|GetFortPawn) :Target self :ContextHandle "()" :InstigatorTags "(GameplayTags=)" :TargetTags "(GameplayTags=)" :EventMagnitude (Variables|Default|GetUpwardLaunchVelocity) :TargetData "()")))

(fn ProjEnter
  (params (Projectile :type Fort%20Projectile%20Base%20Object%20Reference))
  (bind _projectilemovementcomponent (Variables|/Script/FortniteGame.FortProjectileBase|GetProjectileMovementComponent Projectile))
  (bind _projectilegravityscale (Variables|/Script/Engine.ProjectileMovementComponent|GetProjectileGravityScale _projectilemovementcomponent))
  (if (> _projectilegravityscale 0.0)
    (Variables|/Script/Engine.ProjectileMovementComponent|SetProjectileGravityScale (* _projectilegravityscale -1.0) _projectilemovementcomponent)
    (else
      (Variables|/Script/Engine.ProjectileMovementComponent|SetProjectileGravityScale (+ _projectilegravityscale (Variables|Default|GetZeroGravPrj)) _projectilemovementcomponent))))

(fn ProjExit
  (params (Projectile :type Fort%20Projectile%20Base%20Object%20Reference))
  (bind _projectilemovementcomponent (Variables|/Script/FortniteGame.FortProjectileBase|GetProjectileMovementComponent Projectile))
  (bind _projectilegravityscale (Variables|/Script/Engine.ProjectileMovementComponent|GetProjectileGravityScale _projectilemovementcomponent))
  (if (== _projectilegravityscale (Variables|Default|GetZeroGravPrj))
    (Variables|/Script/Engine.ProjectileMovementComponent|SetProjectileGravityScale (- _projectilegravityscale (Variables|Default|GetZeroGravPrj)) _projectilemovementcomponent)
    (else
      (Variables|/Script/Engine.ProjectileMovementComponent|SetProjectileGravityScale (* _projectilegravityscale -1.0) _projectilemovementcomponent))))

(fn ToyEnter
  (params (Projectile :type Physics%20Ball%20Master%20Object%20Reference))
  (bind _projectilemovement (Variables|/Game/Building/ActorBlueprints/Prop/PhysicsBall_Master.PhysicsBall_Master_C|GetProjectileMovement Projectile))
  (bind _projectilegravityscale (Variables|/Script/Engine.ProjectileMovementComponent|GetProjectileGravityScale _projectilemovement))
  (if (> _projectilegravityscale 0.0)
    (Variables|/Script/Engine.ProjectileMovementComponent|SetProjectileGravityScale (* _projectilegravityscale -0.5) _projectilemovement)
    (else
      (Variables|/Script/Engine.ProjectileMovementComponent|SetProjectileGravityScale (+ _projectilegravityscale (Variables|Default|GetZeroGravPrj)) _projectilemovement))))

(fn ToyExit
  (params (Projectile :type Physics%20Ball%20Master%20Object%20Reference))
  (bind _projectilemovement (Variables|/Game/Building/ActorBlueprints/Prop/PhysicsBall_Master.PhysicsBall_Master_C|GetProjectileMovement Projectile))
  (bind _projectilegravityscale (Variables|/Script/Engine.ProjectileMovementComponent|GetProjectileGravityScale _projectilemovement))
  (if (== _projectilegravityscale (Variables|Default|GetZeroGravPrj))
    (Variables|/Script/Engine.ProjectileMovementComponent|SetProjectileGravityScale (- _projectilegravityscale (Variables|Default|GetZeroGravPrj)) _projectilemovement)
    (else
      (Variables|/Script/Engine.ProjectileMovementComponent|SetProjectileGravityScale (* _projectilegravityscale -2.0) _projectilemovement))))

(fn ConstructionScript
  (|CallParentFunction))
